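# Physics asset for the face accessory mesh of the male medium Festive Pug character, covering the tip of its hat.
# Four dyn_hat_tip bodies, three boxes ending in a small capsule, form a chain constrained to the head, with tighter angle limits at the root than along the chain.
# Head and spine_05 capsules provide body volumes, joined by a constraint with free swing and twist.
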PhysicsAsset /BRCosmetics/Characters/Player/Male/Medium/Bodies/M_MED_Festive_Pug/Meshes/Parts/M_MED_Festive_Pug_FaceAcc_Physics
bodies {
 dyn_hat_tip_01 [{shapeType=Box, extentX=7.5544, extentY=11.3612, extentZ=3.1852, center=<4.4125 -0.2302 0.8985>}]
 dyn_hat_tip_02 [{shapeType=Box, extentX=6.566, extentY=9.4671, extentZ=3.8507, center=<4.1692 -0.2333 0.6105>}]
 dyn_hat_tip_03 [{shapeType=Box, extentX=10.4948, extentY=7.3338, extentZ=3.2431, center=<5.7126 -0.1183 0.2707>}]
 dyn_hat_tip_04 [{shapeType=Capsule, length=0.1761, radius=3.3776, center=<2.3199 0.1223 0.2856>}]
 head [{shapeType=Capsule, length=2.3574, radius=13.1196, center=<0.0332 -0.0001 -1.1576>, rotation=<rot 90 0 0>}]
 spine_05 [{shapeType=Capsule, length=12.3109, radius=12.9391}]
}
constraints [
 {
  bone1Name=head
  bone2Name=spine_05
  swing1LimitDegrees=45
  swing1Motion=Free
  swing2LimitDegrees=45
  swing2Motion=Free
  twistLimitDegrees=45
  twistMotion=Free
 }
 {bone1Name=dyn_hat_tip_01, bone2Name=head, swing1LimitDegrees=30, swing2LimitDegrees=30, twistLimitDegrees=10}
 {bone1Name=dyn_hat_tip_02, bone2Name=dyn_hat_tip_01, swing1LimitDegrees=45, swing2LimitDegrees=45, twistLimitDegrees=15}
 {bone1Name=dyn_hat_tip_03, bone2Name=dyn_hat_tip_02, swing1LimitDegrees=45, swing2LimitDegrees=45, twistLimitDegrees=15}
 {bone1Name=dyn_hat_tip_04, bone2Name=dyn_hat_tip_03, swing1LimitDegrees=45, swing2LimitDegrees=45, twistLimitDegrees=15}
]
