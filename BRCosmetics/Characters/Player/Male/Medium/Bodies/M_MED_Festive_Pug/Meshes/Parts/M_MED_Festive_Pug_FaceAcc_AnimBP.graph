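# Animation Blueprint for the Festive Pug face accessory part, adding secondary motion to hoodie strings and other physics bones on top of the pose copied from the attached parent mesh.
# Spine/clavicle constraints feed string AnimDynamics, a spine_05-based RigidBody with skydiving force driving velocity scale, then a fixed right hood-counter offset.
# The four-node hat-tip AnimDynamics chain is not connected to the output.
# The event graph derives head_corrective_alpha from head yaw, but the AnimGraph never reads it.
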AnimBlueprint /BRCosmetics/Characters/Player/Male/Medium/Bodies/M_MED_Festive_Pug/Meshes/Parts/M_MED_Festive_Pug_FaceAcc_AnimBP
target_skeleton=&/M_MED_Festive_Pug_FaceAcc_Skeleton
graphs [
 AnimGraph {nodes [
   n1 = Anim Dynamics - Bone: dyn_hat_tip_01(Alpha=0.5, ExternalForce="0,0,0") {
    angularDampingOverride=0.9999
    angularSpringConstant=3500
    bAngularSpring=true
    bGravityOverrideInSimSpace=true
    bOverrideAngularDamping=true
    gravityScale=0.35
    alphaBoolBlend {blendOption=Linear}
    boundBone {boneName=dyn_hat_tip_01}
    componentAppliedLinearAccClamp=<10000 10000 10000>
    componentLinearAccScale=<0.4 0.4 0.4>
    physicsBodyDefinitions [{
      sphereCollisionRadius=0
      boundBone {boneName=dyn_hat_tip_01}
      boxExtents=<3 6 6>
      constraintSetup {angularLimitsMax=<0 -5 5>, angularLimitsMin=<0 -40 -20>, angularTarget=<1 0 0>}
      localJointOffset=<3 0 0>
     }]
   }
   n2 = Anim Dynamics - Bone: dyn_hat_tip_02(ComponentPose=@n1, Alpha=0.5, ExternalForce="0,0,0") {
    angularDampingOverride=0.9999
    angularSpringConstant=3500
    bAngularSpring=true
    bGravityOverrideInSimSpace=true
    bOverrideAngularDamping=true
    gravityScale=0.35
    simulationSpace=RootRelative
    alphaBoolBlend {blendOption=Linear}
    boundBone {boneName=dyn_hat_tip_02}
    componentAppliedLinearAccClamp=<10000 10000 10000>
    componentLinearAccScale=<0.2 0.2 0.2>
    physicsBodyDefinitions [{
      sphereCollisionRadius=0
      boundBone {boneName=dyn_hat_tip_02}
      boxExtents=<3 6 6>
      constraintSetup {angularLimitsMax=<0 -5 0>, angularLimitsMin=<0 -20 -20>, angularTarget=<1 0 0>}
      localJointOffset=<5 0 0>
     }]
   }
   n3 = Anim Dynamics - Bone: dyn_hat_tip_03(ComponentPose=@n2, Alpha=0.5, ExternalForce="0,0,0") {
    angularDampingOverride=0.9999
    angularSpringConstant=3500
    bAngularSpring=true
    bGravityOverrideInSimSpace=true
    bOverrideAngularDamping=true
    gravityScale=0.35
    simulationSpace=RootRelative
    alphaBoolBlend {blendOption=Linear}
    boundBone {boneName=dyn_hat_tip_03}
    componentAppliedLinearAccClamp=<10000 10000 10000>
    componentLinearAccScale=<0.2 0.2 0.2>
    physicsBodyDefinitions [{
      sphereCollisionRadius=0
      boundBone {boneName=dyn_hat_tip_03}
      boxExtents=<2 5 5>
      constraintSetup {angularLimitsMax=<0 -8 20>, angularLimitsMin=<0 -20 -20>, angularTarget=<1 0 0>}
      localJointOffset=<5 0 0>
     }]
   }
   Anim Dynamics - Bone: dyn_hat_tip_04(ComponentPose=@n3, Alpha=0.5, ExternalForce="0,0,0") {
    angularDampingOverride=1
    angularSpringConstant=3500
    bAngularSpring=true
    bGravityOverrideInSimSpace=true
    bOverrideAngularDamping=true
    gravityScale=0.35
    simulationSpace=RootRelative
    alphaBoolBlend {blendOption=Linear}
    boundBone {boneName=dyn_hat_tip_04}
    componentAppliedLinearAccClamp=<10000 10000 10000>
    componentLinearAccScale=<0.2 0.2 0.2>
    physicsBodyDefinitions [{
      sphereCollisionRadius=0
      boundBone {boneName=dyn_hat_tip_04}
      boxExtents=<2 5 5>
      constraintSetup {angularLimitsMax=<0 0 20>, angularLimitsMin=<0 -20 -20>, angularTarget=<1 0 0>}
      localJointOffset=<5 0 0>
     }]
   }
   n4 = "Constraint - dyn_hoodie_string_01_r (PPPP)"!(
    ComponentPose=Local To Component(LocalPose=Copy Pose From Mesh() {bUseAttachedParent=true})
    Alpha=1
    ConstraintWeights_0=0.2347
    ConstraintWeights_1:float=0.08
    ConstraintWeights_2:float=0.4167
    ConstraintWeights_3:float=0.2685
   ) -> (Pose) {
    alphaBoolBlend {blendOption=Linear}
    boneToModify {boneName=dyn_hoodie_string_01_r}
    constraintSetup [{targetBone {boneName=clavicle_r}}, {targetBone {boneName=spine_03}}, {targetBone {boneName=spine_04}}, {targetBone {boneName=spine_05}}]
    constraintWeights [1,1,1,1]
   }
   n5 = "Constraint - dyn_hoodie_string_01_l (PPPP)"!(
    ComponentPose=@n4
    Alpha=1
    ConstraintWeights_0=0.0796
    ConstraintWeights_1:float=0.4206
    ConstraintWeights_2:float=0.2716
    ConstraintWeights_3:float=0.2282
   ) -> (Pose) {
    alphaBoolBlend {blendOption=Linear}
    boneToModify {boneName=dyn_hoodie_string_01_l}
    constraintSetup [{targetBone {boneName=spine_03}}, {targetBone {boneName=spine_04}}, {targetBone {boneName=spine_05}}, {targetBone {boneName=clavicle_l}}]
    constraintWeights [1,1,1,1]
   }
   n6 = Anim Dynamics - Bone: dyn_hoodie_string_01_r(ComponentPose=@n5, Alpha=0.5, ExternalForce="0,0,0") {
    angularDampingOverride=0.999
    bGravityOverrideInSimSpace=true
    bOverrideAngularDamping=true
    gravityScale=0.3
    simulationSpace=RootRelative
    alphaBoolBlend {blendOption=Linear}
    boundBone {boneName=dyn_hoodie_string_01_r}
    physicsBodyDefinitions [{
      sphereCollisionRadius=0
      boundBone {boneName=dyn_hoodie_string_01_r}
      boxExtents=<5 10 10>
      constraintSetup {angularLimitsMax=<0 20 0>, angularLimitsMin=<0 -20 -40>, angularTarget=<0 0 0>}
      localJointOffset=<5 0 -0.2>
     }]
   }
   n7 = Anim Dynamics - Bone: dyn_hoodie_string_01_l(ComponentPose=@n6, Alpha=0.5, ExternalForce="0,0,0") {
    angularDampingOverride=0.9999
    bGravityOverrideInSimSpace=true
    bOverrideAngularDamping=true
    gravityScale=0.3
    simulationSpace=RootRelative
    alphaBoolBlend {blendOption=Linear}
    boundBone {boneName=dyn_hoodie_string_01_l}
    physicsBodyDefinitions [{
      sphereCollisionRadius=0
      boundBone {boneName=dyn_hoodie_string_01_l}
      boxExtents=<5 10 10>
      constraintSetup {angularLimitsMax=<0 20 0>, angularLimitsMin=<0 -20 -40>, angularTarget=<0 0 0>}
      localJointOffset=<6 0 0.2>
     }]
   }
   n8 = Anim Dynamics - Bone: dyn_hoodie_string_03_r(ComponentPose=@n7, Alpha=0.5, ExternalForce="0,0,0") {
    angularDampingOverride=0.999
    bGravityOverrideInSimSpace=true
    bOverrideAngularDamping=true
    gravityScale=0.3
    simulationSpace=RootRelative
    alphaBoolBlend {blendOption=Linear}
    boundBone {boneName=dyn_hoodie_string_03_r}
    physicsBodyDefinitions [{
      sphereCollisionRadius=0
      boundBone {boneName=dyn_hoodie_string_03_r}
      boxExtents=<5 10 10>
      constraintSetup {angularLimitsMax=<0 20 0>, angularLimitsMin=<0 -20 -40>, angularTarget=<0 0 0>}
      localJointOffset=<5 0 -0.2>
     }]
   }
   n9 = Anim Dynamics - Bone: dyn_hoodie_string_03_l(ComponentPose=@n8, Alpha=0.5, ExternalForce="0,0,0") {
    angularDampingOverride=0.9999
    bGravityOverrideInSimSpace=true
    bOverrideAngularDamping=true
    gravityScale=0.3
    simulationSpace=RootRelative
    alphaBoolBlend {blendOption=Linear}
    boundBone {boneName=dyn_hoodie_string_03_l}
    physicsBodyDefinitions [{
      sphereCollisionRadius=0
      boundBone {boneName=dyn_hoodie_string_03_l}
      boxExtents=<5 10 10>
      constraintSetup {angularLimitsMax=<0 20 0>, angularLimitsMin=<0 -20 -40>, angularTarget=<0 0 0>}
      localJointOffset=<6 0 0.2>
     }]
   }
   n10 = RigidBody!(
    ComponentPose=@n9
    Alpha
    ExternalForce
    ComponentLinearVelScale:Vector=Break Fort Anim Input Skydiving External Force(FortAnimInput_SkydivingExternalForce=Get SkydivingExternalForceInput()) -> (
     bUseSkydivingVectorForce=~
     SkydivingVectorForce=~
     HeadToPelvisDirection=~
     FloatingMultiplier=~
     FloatingAdditive=~
     DivingMultiplier=~
     DivingAdditive=~
     ParachutingMultiplier=~
     ParachutingAdditive=~
    ).SkydivingVectorForce
   ) -> (Pose) {
    bClampLinearTranslationLimitToRefPose=true
    bOverrideWorldGravity=true
    overridePhysicsAsset=&/M_MED_Festive_Pug_FaceAcc_Physics
    simulationSpace=BaseBoneSpace
    alphaBoolBlend {blendOption=Linear}
    baseBoneRef {boneName=spine_05}
    componentAppliedLinearAccClamp=<1000 1000 1000>
    componentLinearAccScale=<0.5 0.5 0.5>
    overrideWorldGravity {z=-980}
   }
   n11 = "Transform (Modify) Bone - Bone: dyn_hood_counter_r"(ComponentPose=@n10, Alpha=1, Translation="-0.000,-1.156,0.846", Rotation="0.000,-0.000,16.379", Scale="1,1,1") {
    rotationMode=BMM_Additive
    rotationSpace=BCS_BoneSpace
    translationMode=BMM_Additive
    translationSpace=BCS_BoneSpace
    alphaBoolBlend {blendOption=Linear}
    boneToModify {boneName=dyn_hood_counter_r}
    rotation {roll=16.379}
    translation {y=-1.156, z=0.846}
   }
   Output Pose(Result=Component To Local(ComponentPose=@n11)) {layerGroup=None}
  ]}
 EventGraph {nodes [
   Event BlueprintUpdateAnimation()
   n1 = BreakTransform(InTransform=GetDeltaTransformFromRefPose(self=GetPartSkeletalMeshComponent(), BoneName=head, BaseName=None))
   > Set head_corrective_alpha(head_corrective_alpha:double=MapRangeClamped(Value=BreakRotator(InRot=@n1.Rotation).Yaw, InRangeA=0, InRangeB=-30)) -> (Output_Get:double)
   TryGetPawnOwner()
  ]}
]
variables [head_corrective_alpha {type=number}]
